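# EQS query for avian NPCs that picks a flight destination up and ahead of the bird, favoring points away from the nearest player.
# Three fallback options generate circle points: an elevated forward arc, an elevated full circle, then a low circle geometry-projected and lifted.
# Each shares tests that reject too-close, near-overhead or blocked points and score player distance and forward alignment.
# A heavily weighted visibility sweep prefers, without requiring, paths clear of obstacles like vehicles.
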EnvQuery /Irwin/Athena/AI/EQS/Behaviors/EQS_NPC_Avian_PointsUpAndAheadOfQuerier
options [
 {
  generator {
   name=EnvQueryGenerator_OnCircle_5
   class=EnvQueryGenerator_OnCircle
   properties {
    arcAngle=90
    bDefineArc=true
    circleCenterZOffset=1550
    numberOfPoints=4
    arcDirection {
     dirMode=Rotation
     lineTo=/Game/Athena/AI/EQS/Contexts/EnvQueryContext_Querier_RandomPointOnUnitCircle_2D.EnvQueryContext_Querier_RandomPointOnUnitCircle_2D_C
    }
    circleRadius {dataBinding=null, dataField=BehaviorDistance, defaultValue=5550}
    projectionData {
     extentX=10
     navigationFilter=&RecastFilter_UseDefaultArea
     postProjectionVerticalOffset=500
     projectDown=100
     projectUp=100
     traceMode=None
    }
    traceData {
     extentX=0
     extentY=5
     extentZ=5
     navigationFilter=&RecastFilter_UseDefaultArea
     traceMode=GeometryByChannel
     traceShape=Box
    }
   }
  }
  tests [
   EnvQueryTest_Distance_12 {
    class=EnvQueryTest_Distance
    properties {
     clampMinType=SpecifiedValue
     distanceTo=&EnvQueryContext_Querier
     filterType=Minimum
     floatValueMin=1500
     scoreClampMin=250
     scoringEquation=Constant
     scoringFactor=1
     testComment=de-prioritize points we are very close to
     testMode=Distance3D
     testPurpose=FilterAndScore
    }
   }
   EnvQueryTest_Distance_13 {
    class=EnvQueryTest_Distance
    properties {
     clampMaxType=SpecifiedValue
     distanceTo=/Game/AI/EQS/Contexts/NearestPlayerContext.NearestPlayerContext_C
     filterType=Range
     floatValueMin=0
     scoreClampMax=6000
     scoringEquation=Linear
     scoringFactor=3
     testComment="Pick point farthest from NearestPlayer (unless NearestPlayer is SUPER far)"
     testMode=Distance3D
     testPurpose=Score
    }
   }
   EnvQueryTest_Dot_2 {
    class=EnvQueryTest_Dot
    properties {clampMaxType=None, clampMinType=None, scoreClampMax=0, scoreClampMin=0, scoringFactor=1, testComment=Favor points ahead of bird}
   }
   EnvQueryTest_Trace_9 {
    class=EnvQueryTest_Trace
    properties {
     scoringFactor=10
     testComment="PREFER things with a clear visibility sweep, but don't require it. Visibility blocks on Vehicles, for instances, the FCC_Trace_World below does NOT. Fixes a lot of weird interactions with vehicles."
     testPurpose=Score
     traceData {extentX=0, extentY=20, extentZ=20, traceShape=Box}
    }
   }
   EnvQueryTest_Trace_10 {
    class=EnvQueryTest_Trace
    properties {
     scoringFactor=1
     testComment=Sweep for clear space
     testPurpose=Filter
     traceData {extentX=0, extentY=20, extentZ=20, serializedChannel=ECC_GameTraceChannel11, traceChannel=TraceTypeQuery10, traceShape=Box}
    }
   }
   EnvQueryTest_Distance_14 {
    class=EnvQueryTest_Distance
    properties {
     distanceTo=&EnvQueryContext_Querier
     filterType=Minimum
     floatValueMin=1000
     scoringEquation=Constant
     scoringFactor=1
     testComment=Throw out points that are more or less directly overhead. We don't want to fly STRAIGHT UP.
     testMode=Distance2D
     testPurpose=Filter
    }
   }
  ]
 }
 {
  generator {
   name=EnvQueryGenerator_OnCircle_3
   class=EnvQueryGenerator_OnCircle
   properties {
    circleCenterZOffset=1550
    numberOfPoints=4
    arcDirection {lineTo=&EnvQueryContext_Querier_RandomPointOnUnitCircle_2D_C}
    circleRadius {dataBinding=null, dataField=BehaviorDistance, defaultValue=5550}
    projectionData {
     extentX=10
     navigationFilter=&RecastFilter_UseDefaultArea
     postProjectionVerticalOffset=500
     projectDown=100
     projectUp=100
     traceMode=None
    }
    traceData {
     extentX=0
     extentY=5
     extentZ=5
     navigationFilter=&RecastFilter_UseDefaultArea
     traceMode=GeometryByChannel
     traceShape=Box
    }
   }
  }
  tests [
   EnvQueryTest_Distance_0 {
    class=EnvQueryTest_Distance
    properties {
     clampMinType=SpecifiedValue
     distanceTo=&EnvQueryContext_Querier
     filterType=Minimum
     floatValueMin=1500
     scoreClampMin=250
     scoringEquation=Constant
     scoringFactor=1
     testComment=de-prioritize points we are very close to
     testMode=Distance3D
     testPurpose=FilterAndScore
    }
   }
   EnvQueryTest_Distance_1 {
    class=EnvQueryTest_Distance
    properties {
     clampMaxType=SpecifiedValue
     distanceTo=&NearestPlayerContext_C
     filterType=Range
     floatValueMin=0
     scoreClampMax=6000
     scoringEquation=Linear
     scoringFactor=3
     testComment="Pick point farthest from NearestPlayer (unless NearestPlayer is SUPER far)"
     testMode=Distance3D
     testPurpose=Score
    }
   }
   EnvQueryTest_Dot_0 {
    class=EnvQueryTest_Dot
    properties {clampMaxType=None, clampMinType=None, scoreClampMax=0, scoreClampMin=0, scoringFactor=1, testComment=Favor points ahead of bird}
   }
   EnvQueryTest_Trace_0 {
    class=EnvQueryTest_Trace
    properties {
     scoringFactor=10
     testComment="PREFER things with a clear visibility sweep, but don't require it. Visibility blocks on Vehicles, for instances, the FCC_Trace_World below does NOT. Fixes a lot of weird interactions with vehicles."
     testPurpose=Score
     traceData {extentX=0, extentY=20, extentZ=20, traceShape=Box}
    }
   }
   EnvQueryTest_Trace_1 {
    class=EnvQueryTest_Trace
    properties {
     scoringFactor=1
     testComment=Sweep for clear space
     testPurpose=Filter
     traceData {extentX=0, extentY=20, extentZ=20, serializedChannel=ECC_GameTraceChannel11, traceChannel=TraceTypeQuery10, traceShape=Box}
    }
   }
   EnvQueryTest_Distance_2 {
    class=EnvQueryTest_Distance
    properties {
     distanceTo=&EnvQueryContext_Querier
     filterType=Minimum
     floatValueMin=1000
     scoringEquation=Constant
     scoringFactor=1
     testComment=Throw out points that are more or less directly overhead. We don't want to fly STRAIGHT UP.
     testMode=Distance2D
     testPurpose=Filter
    }
   }
  ]
 }
 {
  generator {
   name=EnvQueryGenerator_OnCircle_6
   class=EnvQueryGenerator_OnCircle
   properties {
    circleCenterZOffset=10
    numberOfPoints=4
    arcDirection {lineTo=&EnvQueryContext_Querier_RandomPointOnUnitCircle_2D_C}
    circleRadius {dataBinding=null, dataField=BehaviorDistance, defaultValue=5550}
    projectionData {
     extentX=10
     navigationFilter=&RecastFilter_UseDefaultArea
     postProjectionVerticalOffset=1850
     projectDown=100
     projectUp=100
     traceMode=GeometryByChannel
    }
    traceData {
     extentX=0
     extentY=5
     extentZ=5
     navigationFilter=&RecastFilter_UseDefaultArea
     traceMode=GeometryByChannel
     traceShape=Box
    }
   }
  }
  tests [
   EnvQueryTest_Distance_3 {
    class=EnvQueryTest_Distance
    properties {
     clampMinType=SpecifiedValue
     distanceTo=&EnvQueryContext_Querier
     filterType=Minimum
     floatValueMin=1500
     scoreClampMin=250
     scoringEquation=Constant
     scoringFactor=1
     testComment=de-prioritize points we are very close to
     testMode=Distance3D
     testPurpose=FilterAndScore
    }
   }
   EnvQueryTest_Distance_7 {
    class=EnvQueryTest_Distance
    properties {
     clampMaxType=SpecifiedValue
     distanceTo=&NearestPlayerContext_C
     filterType=Range
     floatValueMin=0
     scoreClampMax=6000
     scoringEquation=Linear
     scoringFactor=3
     testComment="Pick point farthest from NearestPlayer (unless NearestPlayer is SUPER far)"
     testMode=Distance3D
     testPurpose=Score
    }
   }
   EnvQueryTest_Dot_1 {
    class=EnvQueryTest_Dot
    properties {clampMaxType=None, clampMinType=None, scoreClampMax=0, scoreClampMin=0, scoringFactor=1, testComment=Favor points ahead of bird}
   }
   EnvQueryTest_Trace_5 {
    class=EnvQueryTest_Trace
    properties {
     scoringFactor=10
     testComment="PREFER things with a clear visibility sweep, but don't require it. Visibility blocks on Vehicles, for instances, the FCC_Trace_World below does NOT. Fixes a lot of weird interactions with vehicles."
     testPurpose=Score
     traceData {extentX=0, extentY=20, extentZ=20, traceShape=Box}
    }
   }
   EnvQueryTest_Trace_6 {
    class=EnvQueryTest_Trace
    properties {
     scoringFactor=1
     testComment=Sweep for clear space
     testPurpose=Filter
     traceData {extentX=0, extentY=20, extentZ=20, serializedChannel=ECC_GameTraceChannel11, traceChannel=TraceTypeQuery10, traceShape=Box}
    }
   }
   EnvQueryTest_Distance_8 {
    class=EnvQueryTest_Distance
    properties {
     distanceTo=&EnvQueryContext_Querier
     filterType=Minimum
     floatValueMin=1000
     scoringEquation=Constant
     scoringFactor=1
     testComment=Throw out points that are more or less directly overhead. We don't want to fly STRAIGHT UP.
     testMode=Distance2D
     testPurpose=Filter
    }
   }
  ]
 }
]
