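# EQS query that picks a navigable point for an NPC to move to land, steering away from water.
# Both options generate an ability-bound ring of navmesh points using water-avoiding filters, then test pathfinding, path cost, querier facing and slight randomness.
# The first strictly excludes water via a projection test and favors nearer 3D points as likely level ground.
# The fallback drops that exclusion and favors the farthest 2D point.
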EnvQuery /Irwin/Athena/AI/EQS/Behaviors/EQS_NPC_MoveToLand
options [
 {
  generator {
   name=EnvQueryGenerator_OnCircle_1
   class=EnvQueryGenerator_OnCircle
   properties {
    circleRadius {
     dataBinding=&self:EnvQueryGenerator_OnCircle_1.FortAIDataProvider_Ability_2
     dataField=BehaviorDistance
     defaultValue=2500
    }
    numberOfPoints {
     dataBinding=&self:EnvQueryGenerator_OnCircle_1.FortAIDataProvider_Ability_1
     dataField=None
     defaultValue=12
    }
    projectionData {
     extentX=10
     navigationFilter=&NavFilter_NPC_AvoidAllWater_C
     postProjectionVerticalOffset=30
     projectDown=1024
    }
    traceData {extentX=0, navigationFilter=&NavFilter_NPC_AvoidAllWater_C}
   }
  }
  tests [
   EnvQueryTest_Project_0 {
    class=EnvQueryTest_Project
    properties {
     scoringFactor=1
     testComment=""
     testPurpose=Filter
     projectionData {
      navigationFilter=&NavFilter_NPC_ExcludeAllWater_C
      postProjectionVerticalOffset=0
      projectDown=1024
      projectUp=1024
      traceMode=Navigation
     }
    }
   }
   EnvQueryTest_Dot_0 {
    class=EnvQueryTest_Dot
    properties {
     clampMaxType=None
     clampMinType=None
     scoreClampMax=0
     scoreClampMin=0
     testComment=querier facing item
     scoringFactor {dataBinding=null, dataField=FloatValue, defaultValue=0.2}
    }
   }
   EnvQueryTest_Distance_1 {
    class=EnvQueryTest_Distance
    properties {
     distanceTo=&EnvQueryContext_Querier
     filterType=Range
     floatValueMin=0
     scoringEquation=InverseLinear
     testComment="Farther will likely mean deviation in height (land)"
     testMode=Distance3D
     testPurpose=Score
     scoringFactor {dataBinding=null, dataField=FloatValue, defaultValue=1}
    }
   }
   EnvQueryTest_PathfindingBatch_0 {
    class=EnvQueryTest_PathfindingBatch
    properties {filterClass=&NavFilter_NPC_AvoidAllWater_C}
   }
   EnvQueryTest_PathfindingBatch_1 {
    class=EnvQueryTest_PathfindingBatch
    properties {
     filterClass=&NavFilter_NPC_AvoidAllWater_C
     filterType=Minimum
     testComment=Score path
     testMode=PathCost
     testPurpose=Score
    }
   }
   FortQueryTest_Random_0 {class=FortQueryTest_Random, properties {floatValueMax=0.1, scoringFactor {dataBinding=null, dataField=FloatValue, defaultValue=0.2}}}
  ]
 }
 {
  generator {
   name=EnvQueryGenerator_OnCircle_0
   class=EnvQueryGenerator_OnCircle
   properties {
    circleRadius {
     dataBinding=&self:EnvQueryGenerator_OnCircle_0.FortAIDataProvider_Ability_2
     dataField=BehaviorDistance
     defaultValue=2500
    }
    numberOfPoints {
     dataBinding=&self:EnvQueryGenerator_OnCircle_0.FortAIDataProvider_Ability_1
     dataField=None
     defaultValue=12
    }
    projectionData {
     extentX=10
     navigationFilter=&NavFilter_NPC_AvoidAllWater_C
     postProjectionVerticalOffset=30
     projectDown=1024
    }
    traceData {extentX=0, navigationFilter=&NavFilter_NPC_AvoidAllWater_C}
   }
  }
  tests [
   EnvQueryTest_PathfindingBatch_2 {
    class=EnvQueryTest_PathfindingBatch
    properties {filterClass=&NavFilter_NPC_AvoidAllWater_C}
   }
   EnvQueryTest_PathfindingBatch_3 {
    class=EnvQueryTest_PathfindingBatch
    properties {
     filterClass=&NavFilter_NPC_AvoidAllWater_C
     filterType=Minimum
     testComment=Score path
     testMode=PathCost
     testPurpose=Score
    }
   }
   EnvQueryTest_Distance_0 {
    class=EnvQueryTest_Distance
    properties {
     distanceTo=&EnvQueryContext_Querier
     filterType=Range
     floatValueMin=0
     scoringEquation=Linear
     testComment=Pick point farthest from Self
     testMode=Distance2D
     testPurpose=Score
     scoringFactor {dataBinding=null, dataField=FloatValue, defaultValue=1}
    }
   }
   EnvQueryTest_Dot_1 {
    class=EnvQueryTest_Dot
    properties {
     clampMaxType=None
     clampMinType=None
     scoreClampMax=0
     scoreClampMin=0
     testComment=querier facing item
     scoringFactor {dataBinding=null, dataField=FloatValue, defaultValue=0.4}
    }
   }
   FortQueryTest_Random_1 {class=FortQueryTest_Random, properties {floatValueMax=0.1, scoringFactor {dataBinding=null, dataField=FloatValue, defaultValue=0.1}}}
  ]
 }
]
