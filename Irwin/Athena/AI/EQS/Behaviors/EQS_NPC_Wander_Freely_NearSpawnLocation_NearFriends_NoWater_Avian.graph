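# EQS query that picks wander destinations for avian NPCs, favoring spots near their spawn location and longer flights.
# Three ordered options try points on building actors, then a grid plus circle, then that again with the path trace only scored.
# Tests require projectable surfaces with upward clearance and reject spots near other AI or their destinations.
# An inverted water-only nav projection excludes water while keeping off-nav points valid; traps, destructible surfaces and randomness adjust scores.
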EnvQuery /Irwin/Athena/AI/EQS/Behaviors/EQS_NPC_Wander_Freely_NearSpawnLocation_NearFriends_NoWater_Avian
options [
 {
  generator {
   name=EnvQueryGenerator_Composite_1
   class=EnvQueryGenerator_Composite
   properties {
    optionName=""
    generators [&self:EnvQueryGenerator_Composite_1.FortQueryGenerator_PointsOnBuildingActors_1]
   }
  }
  tests [
   EnvQueryTest_Distance_6 {
    class=EnvQueryTest_Distance
    properties {
     clampMinType=SpecifiedValue
     distanceTo=&EnvQueryContext_Querier
     filterType=Minimum
     floatValueMin=100
     scoreClampMin=1500
     scoringEquation=Constant
     scoringFactor=1
     testComment=Throw out points we're basically on top of. Prefer things a min distance away for a longer flight
     testMode=Distance3D
     testPurpose=FilterAndScore
    }
   }
   EnvQueryTest_Project_4 {
    class=EnvQueryTest_Project
    properties {
     scoringFactor=0
     testComment="Can we project to a surface? We USED to do this in Generators, but cheaper here since it can be time sliced."
     testPurpose=Filter
     projectionData {
      extentY=20
      extentZ=20
      navigationFilter=null
      postProjectionVerticalOffset=5
      projectDown=3000
      projectUp=3000
      serializedChannel=ECC_GameTraceChannel11
      traceChannel=TraceTypeQuery10
      traceMode=GeometryByChannel
      traceShape=Box
     }
    }
   }
   EnvQueryTest_Project_5 {
    class=EnvQueryTest_Project
    properties {
     boolValue=false
     scoringFactor=1
     testComment=Require clear upward sweep for pawn capsule in fallback despan situations. The Generator itself is not doing this well enough.
     testPurpose=Filter
     projectionData {
      extentY=20
      extentZ=20
      navigationFilter=null
      postProjectionVerticalOffset=0
      projectDown=0
      projectUp=3000
      serializedChannel=ECC_GameTraceChannel11
      traceChannel=TraceTypeQuery10
      traceMode=GeometryByChannel
      traceShape=Box
     }
    }
   }
   EnvQueryTest_Project_9 {
    class=EnvQueryTest_Project
    properties {
     boolValue=false
     scoringFactor=1
     testComment=Baseline Score - Exclude Water - We do this backwards so other generated points that are NOT on Nav will still be Valid even if they can't project to nav.
     testPurpose=FilterAndScore
     projectionData {
      navigationFilter=/Game/Athena/AI/NavFilters/NavFilter_NPC_WaterOnly.NavFilter_NPC_WaterOnly_C
      postProjectionVerticalOffset=0
      projectDown=3000
      projectUp=3000
      traceMode=Navigation
     }
    }
   }
   EnvQueryTest_Project_11 {
    class=EnvQueryTest_Project
    properties {
     boolValue=false
     scoringFactor=0.2
     testComment=Was 0.6 - Secondary score for things on destructible surfaces.
     testPurpose=Score
     projectionData {
      navigationFilter=null
      postProjectionVerticalOffset=5
      projectDown=20
      projectUp=20
      serializedChannel=ECC_GameTraceChannel6
      traceChannel=TraceTypeQuery7
      traceMode=GeometryByChannel
     }
    }
   }
   EnvQueryTest_Distance_7 {
    class=EnvQueryTest_Distance
    properties {
     clampMaxType=SpecifiedValue
     clampMinType=SpecifiedValue
     distanceTo=/Script/FortniteAIServer.FortQueryContext_AIPawnSpawnLocation
     filterType=Range
     floatValueMin=0
     scoringEquation=InverseLinear
     testComment=Ring around Spawn Location
     testMode=Distance2D
     testPurpose=Score
     scoreClampMax {
      dataBinding=&self:EnvQueryTest_Distance_7.FortAIDataProvider_Ability_1
      dataField=BehaviorDistance
      defaultValue=2500
     }
     scoreClampMin {
      dataBinding=&self:EnvQueryTest_Distance_7.FortAIDataProvider_Ability_0
      dataField=BehaviorDistance
      defaultValue=1500
     }
     scoringFactor {dataBinding=null, dataField=FloatValue, defaultValue=10}
    }
   }
   FortQueryTest_Random_1 {
    class=FortQueryTest_Random
    properties {
     floatValueMax=0.1
     testComment=Was 0.6 - Randomize a bit to break-up
     scoringFactor {dataBinding=null, dataField=FloatValue, defaultValue=0.6}
    }
   }
   EnvQueryTest_Distance_8 {
    class=EnvQueryTest_Distance
    properties {
     distanceTo=&FortQueryContext_NearbyAIPawns
     filterType=Minimum
     floatValueMin=250
     scoringEquation=InverseLinear
     scoringFactor=3
     testComment=Try not to overlap AI Pawns
     testMode=Distance3D
     testPurpose=Filter
    }
   }
   EnvQueryTest_Distance_9 {
    class=EnvQueryTest_Distance
    properties {
     distanceTo=&FortQueryContext_NearbyAIPawnsMoveDestinations
     filterType=Minimum
     floatValueMin=250
     scoringEquation=Linear
     scoringFactor=1
     testComment=not near other pawn move destinations
     testMode=Distance3D
     testPurpose=Filter
    }
   }
   EnvQueryTest_Trace_4 {
    class=EnvQueryTest_Trace
    properties {
     scoringFactor=0
     testComment=Straight sweep from AI to Destination. Looking for a roughly open path.
     testPurpose=Filter
     traceData {extentX=5, extentZ=5}
    }
   }
   EnvQueryTest_Distance_12 {
    class=EnvQueryTest_Distance
    properties {
     clampMinType=SpecifiedValue
     distanceTo=/Game/Athena/AI/EQS/Contexts/EnvQueryContext_BuildingTraps.EnvQueryContext_BuildingTraps_C
     filterType=Minimum
     floatValueMin=370
     multipleContextScoreOp=MinScore
     scoreClampMin=370
     scoringEquation=Constant
     scoringFactor=2
     testComment=Increase score for things NOT near Traps.
     testMode=Distance3D
     testPurpose=Score
    }
   }
  ]
 }
 {
  generator {
   name=EnvQueryGenerator_Composite_2
   class=EnvQueryGenerator_Composite
   properties {
    optionName=""
    generators [
     &self:EnvQueryGenerator_Composite_2.EnvQueryGenerator_SimpleGrid_1
     &self:EnvQueryGenerator_Composite_2.EnvQueryGenerator_OnCircle_0
    ]
   }
  }
  tests [
   EnvQueryTest_Distance_1 {
    class=EnvQueryTest_Distance
    properties {
     clampMinType=SpecifiedValue
     distanceTo=&EnvQueryContext_Querier
     filterType=Minimum
     floatValueMin=100
     scoreClampMin=1500
     scoringEquation=Constant
     scoringFactor=1
     testComment=Throw out points we're basically on top of. Prefer things a min distance away for a longer flight
     testMode=Distance3D
     testPurpose=FilterAndScore
    }
   }
   EnvQueryTest_Project_0 {
    class=EnvQueryTest_Project
    properties {
     scoringFactor=0
     testComment="Can we project to a surface? We USED to do this in Generators, but cheaper here since it can be time sliced."
     testPurpose=Filter
     projectionData {
      extentY=20
      extentZ=20
      navigationFilter=null
      postProjectionVerticalOffset=5
      projectDown=3000
      projectUp=3000
      serializedChannel=ECC_GameTraceChannel11
      traceChannel=TraceTypeQuery10
      traceMode=GeometryByChannel
      traceShape=Box
     }
    }
   }
   EnvQueryTest_Project_1 {
    class=EnvQueryTest_Project
    properties {
     boolValue=false
     scoringFactor=1
     testComment=Require clear upward sweep for pawn capsule in fallback despan situations. The Generator itself is not doing this well enough.
     testPurpose=Filter
     projectionData {
      extentY=20
      extentZ=20
      navigationFilter=null
      postProjectionVerticalOffset=0
      projectDown=0
      projectUp=3000
      serializedChannel=ECC_GameTraceChannel11
      traceChannel=TraceTypeQuery10
      traceMode=GeometryByChannel
      traceShape=Box
     }
    }
   }
   EnvQueryTest_Project_2 {
    class=EnvQueryTest_Project
    properties {
     boolValue=false
     scoringFactor=1
     testComment=Baseline Score - Exclude Water - We do this backwards so other generated points that are NOT on Nav will still be Valid even if they can't project to nav.
     testPurpose=FilterAndScore
     projectionData {
      navigationFilter=&NavFilter_NPC_WaterOnly_C
      postProjectionVerticalOffset=0
      projectDown=3000
      projectUp=3000
      traceMode=Navigation
     }
    }
   }
   EnvQueryTest_Project_3 {
    class=EnvQueryTest_Project
    properties {
     boolValue=false
     scoringFactor=0.2
     testComment=Was 0.6 - Secondary score for things on destructible surfaces.
     testPurpose=Score
     projectionData {
      navigationFilter=null
      postProjectionVerticalOffset=5
      projectDown=20
      projectUp=20
      serializedChannel=ECC_GameTraceChannel6
      traceChannel=TraceTypeQuery7
      traceMode=GeometryByChannel
     }
    }
   }
   EnvQueryTest_Distance_2 {
    class=EnvQueryTest_Distance
    properties {
     clampMaxType=SpecifiedValue
     clampMinType=SpecifiedValue
     distanceTo=&FortQueryContext_AIPawnSpawnLocation
     filterType=Range
     floatValueMin=0
     scoringEquation=InverseLinear
     testComment=Ring around Spawn Location
     testMode=Distance2D
     testPurpose=Score
     scoreClampMax {
      dataBinding=&self:EnvQueryTest_Distance_2.FortAIDataProvider_Ability_1
      dataField=BehaviorDistance
      defaultValue=2500
     }
     scoreClampMin {
      dataBinding=&self:EnvQueryTest_Distance_2.FortAIDataProvider_Ability_0
      dataField=BehaviorDistance
      defaultValue=1500
     }
     scoringFactor {dataBinding=null, dataField=FloatValue, defaultValue=10}
    }
   }
   FortQueryTest_Random_0 {
    class=FortQueryTest_Random
    properties {
     floatValueMax=0.1
     testComment=Was 0.6 - Randomize a bit to break-up
     scoringFactor {dataBinding=null, dataField=FloatValue, defaultValue=0.6}
    }
   }
   EnvQueryTest_Distance_3 {
    class=EnvQueryTest_Distance
    properties {
     distanceTo=&FortQueryContext_NearbyAIPawns
     filterType=Minimum
     floatValueMin=250
     scoringEquation=InverseLinear
     scoringFactor=3
     testComment=Try not to overlap AI Pawns
     testMode=Distance3D
     testPurpose=Filter
    }
   }
   EnvQueryTest_Distance_4 {
    class=EnvQueryTest_Distance
    properties {
     distanceTo=&FortQueryContext_NearbyAIPawnsMoveDestinations
     filterType=Minimum
     floatValueMin=250
     scoringEquation=Linear
     scoringFactor=1
     testComment=not near other pawn move destinations
     testMode=Distance3D
     testPurpose=Filter
    }
   }
   EnvQueryTest_Trace_0 {
    class=EnvQueryTest_Trace
    properties {
     scoringFactor=0
     testComment=Straight sweep from AI to Destination. Looking for a roughly open path.
     testPurpose=Filter
     traceData {extentX=5, extentZ=5}
    }
   }
   EnvQueryTest_Distance_5 {
    class=EnvQueryTest_Distance
    properties {
     clampMinType=SpecifiedValue
     distanceTo=&EnvQueryContext_BuildingTraps_C
     filterType=Minimum
     floatValueMin=370
     multipleContextScoreOp=MinScore
     scoreClampMin=370
     scoringEquation=Constant
     scoringFactor=2
     testComment=Increase score for things NOT near Traps.
     testMode=Distance3D
     testPurpose=Score
    }
   }
  ]
 }
 {
  generator {
   name=EnvQueryGenerator_Composite_3
   class=EnvQueryGenerator_Composite
   properties {
    optionName=""
    generators [
     &self:EnvQueryGenerator_Composite_3.EnvQueryGenerator_SimpleGrid_1
     &self:EnvQueryGenerator_Composite_3.EnvQueryGenerator_OnCircle_0
    ]
   }
  }
  tests [
   EnvQueryTest_Distance_15 {
    class=EnvQueryTest_Distance
    properties {
     clampMinType=SpecifiedValue
     distanceTo=&EnvQueryContext_Querier
     filterType=Minimum
     floatValueMin=100
     scoreClampMin=1500
     scoringEquation=Constant
     scoringFactor=1
     testComment=Throw out points we're basically on top of. Prefer things a min distance away for a longer flight
     testMode=Distance3D
     testPurpose=FilterAndScore
    }
   }
   EnvQueryTest_Project_12 {
    class=EnvQueryTest_Project
    properties {
     scoringFactor=0
     testComment="Can we project to a surface? We USED to do this in Generators, but cheaper here since it can be time sliced."
     testPurpose=Filter
     projectionData {
      extentY=20
      extentZ=20
      navigationFilter=null
      postProjectionVerticalOffset=5
      projectDown=3000
      projectUp=3000
      serializedChannel=ECC_GameTraceChannel11
      traceChannel=TraceTypeQuery10
      traceMode=GeometryByChannel
      traceShape=Box
     }
    }
   }
   EnvQueryTest_Project_13 {
    class=EnvQueryTest_Project
    properties {
     boolValue=false
     scoringFactor=1
     testComment=Require clear upward sweep for pawn capsule in fallback despan situations. The Generator itself is not doing this well enough.
     testPurpose=Filter
     projectionData {
      extentY=20
      extentZ=20
      navigationFilter=null
      postProjectionVerticalOffset=0
      projectDown=0
      projectUp=3000
      serializedChannel=ECC_GameTraceChannel11
      traceChannel=TraceTypeQuery10
      traceMode=GeometryByChannel
      traceShape=Box
     }
    }
   }
   EnvQueryTest_Project_14 {
    class=EnvQueryTest_Project
    properties {
     boolValue=false
     scoringFactor=1
     testComment=Baseline Score - Exclude Water - We do this backwards so other generated points that are NOT on Nav will still be Valid even if they can't project to nav.
     testPurpose=FilterAndScore
     projectionData {
      navigationFilter=&NavFilter_NPC_WaterOnly_C
      postProjectionVerticalOffset=0
      projectDown=3000
      projectUp=3000
      traceMode=Navigation
     }
    }
   }
   EnvQueryTest_Project_15 {
    class=EnvQueryTest_Project
    properties {
     boolValue=false
     scoringFactor=0.2
     testComment=Was 0.6 - Secondary score for things on destructible surfaces.
     testPurpose=Score
     projectionData {
      navigationFilter=null
      postProjectionVerticalOffset=5
      projectDown=20
      projectUp=20
      serializedChannel=ECC_GameTraceChannel6
      traceChannel=TraceTypeQuery7
      traceMode=GeometryByChannel
     }
    }
   }
   EnvQueryTest_Distance_16 {
    class=EnvQueryTest_Distance
    properties {
     clampMaxType=SpecifiedValue
     clampMinType=SpecifiedValue
     distanceTo=&FortQueryContext_AIPawnSpawnLocation
     filterType=Range
     floatValueMin=0
     scoringEquation=InverseLinear
     testComment=Ring around Spawn Location
     testMode=Distance2D
     testPurpose=Score
     scoreClampMax {
      dataBinding=&self:EnvQueryTest_Distance_16.FortAIDataProvider_Ability_1
      dataField=BehaviorDistance
      defaultValue=2500
     }
     scoreClampMin {
      dataBinding=&self:EnvQueryTest_Distance_16.FortAIDataProvider_Ability_0
      dataField=BehaviorDistance
      defaultValue=1500
     }
     scoringFactor {dataBinding=null, dataField=FloatValue, defaultValue=10}
    }
   }
   FortQueryTest_Random_3 {
    class=FortQueryTest_Random
    properties {
     floatValueMax=0.1
     testComment=Was 0.6 - Randomize a bit to break-up
     scoringFactor {dataBinding=null, dataField=FloatValue, defaultValue=0.6}
    }
   }
   EnvQueryTest_Distance_17 {
    class=EnvQueryTest_Distance
    properties {
     distanceTo=&FortQueryContext_NearbyAIPawns
     filterType=Minimum
     floatValueMin=250
     scoringEquation=InverseLinear
     scoringFactor=3
     testComment=Try not to overlap AI Pawns
     testMode=Distance3D
     testPurpose=Filter
    }
   }
   EnvQueryTest_Distance_18 {
    class=EnvQueryTest_Distance
    properties {
     distanceTo=&FortQueryContext_NearbyAIPawnsMoveDestinations
     filterType=Minimum
     floatValueMin=250
     scoringEquation=Linear
     scoringFactor=1
     testComment=not near other pawn move destinations
     testMode=Distance3D
     testPurpose=Filter
    }
   }
   EnvQueryTest_Trace_3 {
    class=EnvQueryTest_Trace
    properties {
     scoringFactor=10
     testComment=Straight sweep from AI to Destination. Looking for a roughly open path.
     testPurpose=Score
     traceData {extentX=5, extentZ=5}
    }
   }
   EnvQueryTest_Distance_19 {
    class=EnvQueryTest_Distance
    properties {
     clampMinType=SpecifiedValue
     distanceTo=&EnvQueryContext_BuildingTraps_C
     filterType=Minimum
     floatValueMin=370
     multipleContextScoreOp=MinScore
     scoreClampMin=370
     scoringEquation=Constant
     scoringFactor=2
     testComment=Increase score for things NOT near Traps.
     testMode=Distance3D
     testPurpose=Score
    }
   }
  ]
 }
]
