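# Environment query that picks a wander point very near the goal for terrestrial NPCs, avoiding water.
# It generates a small circle of points around the goal, with point count bound to a game-difficulty curve.
# Water-excluding pathfinding and minimum querier-distance tests filter candidates.
# Scoring favors points ahead but not directly in front, toward the goal and near it, with dot-product clamps bound to difficulty curves.
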EnvQuery /Irwin/Athena/AI/EQS/Behaviors/EQS_NPC_Wander_VeryNearGoal_Terrestrial
options [{
  generator {
   name=EnvQueryGenerator_OnCircle_1
   class=EnvQueryGenerator_OnCircle
   properties {
    circleCenter=&FortQueryContext_Goal
    circleRadius {dataBinding=null, dataField=FloatValue, defaultValue=300}
    numberOfPoints {
     dataBinding=&self:EnvQueryGenerator_OnCircle_1.FortAIDataProvider_FloatCurveOverGameDifficulty_1
     dataField=None
     defaultValue=12
    }
    projectionData {extentX=0, navigationFilter=null, postProjectionVerticalOffset=0, projectDown=1024, traceMode=None}
    traceData {extentX=0, navigationFilter=&NavFilter_NPC_ExcludeAllWater_C}
   }
  }
  tests [
   EnvQueryTest_PathfindingBatch_2 {
    class=EnvQueryTest_PathfindingBatch
    properties {filterClass=&NavFilter_NPC_ExcludeAllWater_C}
   }
   EnvQueryTest_Distance_3 {
    class=EnvQueryTest_Distance
    properties {
     distanceTo=&EnvQueryContext_Querier
     filterType=Minimum
     scoringEquation=InverseLinear
     scoringFactor=1
     testComment=Throw out point I'm right on top of
     testMode=Distance2D
     testPurpose=Filter
     floatValueMin {dataBinding=null, dataField=FloatValue, defaultValue=50}
    }
   }
   EnvQueryTest_Dot_1 {
    class=EnvQueryTest_Dot
    properties {
     clampMaxType=SpecifiedValue
     clampMinType=SpecifiedValue
     referenceValue=0.707
     testComment=Prefer in front of Self
     scoreClampMax {
      dataBinding=&self:EnvQueryTest_Dot_1.FortAIDataProvider_FloatCurveOverGameDifficulty_1
      dataField=FloatValue
      defaultValue=0.4
     }
     scoreClampMin {
      dataBinding=&self:EnvQueryTest_Dot_1.FortAIDataProvider_FloatCurveOverGameDifficulty_0
      dataField=FloatValue
      defaultValue=0
     }
     scoringFactor {dataBinding=null, dataField=FloatValue, defaultValue=0.25}
    }
   }
   EnvQueryTest_Dot_3 {
    class=EnvQueryTest_Dot
    properties {
     clampMaxType=SpecifiedValue
     clampMinType=SpecifiedValue
     floatValueMax=1
     floatValueMin=0.95
     referenceValue=0.707
     scoringEquation=InverseLinear
     testComment=But not DIRECTLY in front of self
     scoreClampMax {
      dataBinding=&self:EnvQueryTest_Dot_3.FortAIDataProvider_FloatCurveOverGameDifficulty_1
      dataField=FloatValue
      defaultValue=1
     }
     scoreClampMin {
      dataBinding=&self:EnvQueryTest_Dot_3.FortAIDataProvider_FloatCurveOverGameDifficulty_0
      dataField=FloatValue
      defaultValue=0.9
     }
     scoringFactor {dataBinding=null, dataField=FloatValue, defaultValue=0.25}
    }
   }
   EnvQueryTest_Dot_0 {
    class=EnvQueryTest_Dot
    properties {
     clampMaxType=SpecifiedValue
     clampMinType=SpecifiedValue
     testComment=Prefer in direction of Goal
     lineA {dirMode=TwoPoints, lineFrom=&EnvQueryContext_Querier, lineTo=&EnvQueryContext_Item}
     lineB {lineTo=&FortQueryContext_Goal}
     scoreClampMax {
      dataBinding=&self:EnvQueryTest_Dot_0.FortAIDataProvider_FloatCurveOverGameDifficulty_1
      dataField=FloatValue
      defaultValue=1
     }
     scoreClampMin {
      dataBinding=&self:EnvQueryTest_Dot_0.FortAIDataProvider_FloatCurveOverGameDifficulty_0
      dataField=FloatValue
      defaultValue=-1
     }
     scoringFactor {dataBinding=null, dataField=FloatValue, defaultValue=0.5}
    }
   }
   EnvQueryTest_Distance_5 {
    class=EnvQueryTest_Distance
    properties {
     distanceTo=&FortQueryContext_Goal
     filterType=Range
     floatValueMin=0
     scoringEquation=InverseLinear
     scoringFactor=0.3
     testComment=Prefer points near the goal
     testMode=Distance2D
     testPurpose=Score
    }
   }
  ]
 }]
